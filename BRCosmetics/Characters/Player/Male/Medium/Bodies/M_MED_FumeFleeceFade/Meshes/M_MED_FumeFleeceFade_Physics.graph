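# Physics asset for the FumeFleeceFade male medium player body, set up around a dangling holster bone chain and the body parts near it.
# Capsules cover the pelvis, spine and left thigh, joined by angularly free constraints.
# The holster master hangs freely off spine_01 and leads a three-bone chain with tight angular limits, where one swing axis widens toward the tip.
# The last holster bone carries two long, offset capsules.
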PhysicsAsset /BRCosmetics/Characters/Player/Male/Medium/Bodies/M_MED_FumeFleeceFade/Meshes/M_MED_FumeFleeceFade_Physics
bodies {
 dyn_holster_01 [{shapeType=Capsule, length=2, radius=2, shapeName=dyn_holster_01_capsule, center=<2 0 0>}]
 dyn_holster_02 [{shapeType=Capsule, length=2, radius=2, shapeName=dyn_holster_02_capsule, center=<2 0 0>}]
 dyn_holster_03 [
  {
   shapeType=Capsule
   length=30.0371
   radius=2.1599
   shapeName=dyn_holster_03_capsule
   center=<3.3698 -2.884 -1.337>
   rotation=<rot 0.0001 88.1517 -76.9178>
  }
  {shapeType=Capsule, length=30, radius=2, shapeName=dyn_holster_03_capsule, center=<3.2826 -2.7523 2.4952>, rotation=<rot 0 88 -89.8812>}
 ]
 dyn_holster_MSTR [{shapeType=Capsule, length=0.505, radius=0.505, shapeName=dyn_holster_MSTR_capsule, rotation=<rot 0 -90 0>}]
 pelvis [{shapeType=Capsule, length=0.727, radius=15.6013, rotation=<rot 90 0 0>}]
 spine_01 [{shapeType=Capsule, length=0.3099, radius=14.0431, shapeName=spine_01_capsule, rotation=<rot 90 0 0>}]
 spine_02 [{shapeType=Capsule, length=0.659, radius=12.9822, shapeName=spine_02_capsule, rotation=<rot 90 0 0>}]
 spine_03 [{shapeType=Capsule, length=0.5789, radius=12.2693, shapeName=spine_03_capsule, rotation=<rot 90 0 0>}]
 spine_04 [{shapeType=Capsule, length=1.5723, radius=13.6346, shapeName=spine_04_capsule}]
 thigh_l [{shapeType=Capsule, length=31.7954, radius=7.6116, shapeName=thigh_l_capsule, center=<-18.5919 0 0>, rotation=<rot 90 0 0>}]
}
constraints [
 {
  bone1Name=spine_01
  bone2Name=pelvis
  swing1LimitDegrees=45
  swing1Motion=Free
  swing2LimitDegrees=45
  swing2Motion=Free
  twistLimitDegrees=45
  twistMotion=Free
 }
 {
  bone1Name=spine_02
  bone2Name=spine_01
  swing1LimitDegrees=45
  swing1Motion=Free
  swing2LimitDegrees=45
  swing2Motion=Free
  twistLimitDegrees=45
  twistMotion=Free
 }
 {
  bone1Name=spine_03
  bone2Name=spine_02
  swing1LimitDegrees=45
  swing1Motion=Free
  swing2LimitDegrees=45
  swing2Motion=Free
  twistLimitDegrees=45
  twistMotion=Free
 }
 {
  bone1Name=spine_04
  bone2Name=spine_03
  swing1LimitDegrees=45
  swing1Motion=Free
  swing2LimitDegrees=45
  swing2Motion=Free
  twistLimitDegrees=45
  twistMotion=Free
 }
 {
  bone1Name=dyn_holster_MSTR
  bone2Name=spine_01
  swing1LimitDegrees=45
  swing1Motion=Free
  swing2LimitDegrees=45
  swing2Motion=Free
  twistLimitDegrees=45
  twistMotion=Free
 }
 {bone1Name=dyn_holster_01, bone2Name=dyn_holster_MSTR, swing1LimitDegrees=4, swing2LimitDegrees=5, twistLimitDegrees=3}
 {bone1Name=dyn_holster_02, bone2Name=dyn_holster_01, swing1LimitDegrees=4, swing2LimitDegrees=15, twistLimitDegrees=3}
 {bone1Name=dyn_holster_03, bone2Name=dyn_holster_02, swing1LimitDegrees=4, swing2LimitDegrees=20, twistLimitDegrees=3}
 {
  bone1Name=thigh_l
  bone2Name=pelvis
  swing1LimitDegrees=45
  swing1Motion=Free
  swing2LimitDegrees=45
  swing2Motion=Free
  twistLimitDegrees=45
  twistMotion=Free
 }
]
